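# Niagara skydiving contrail for a Kpop cosmetic, trailing ribbons from hand and foot skeletal mesh bones with glow and star sprites.
# Both ribbon emitters burst a fixed particle set, partitioned into per-bone ribbons drawn by several layered renderers.
# Each update re-pins particles to their bones, then offsets them against DiveVelocity by normalized index, forming tails without continuous spawning.
# Glow sprites follow hand-mesh bones with color cycling, while short-lived Stars spawn in cylinders around them.
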NiagaraSystem /BRCosmetics/Athena/Effects/Particles/Skydiving/NS_Contrail_Kpop/NS_Contrail_Kpop_Converted
dependencies {
 usedDataInterfaces [&NiagaraDataInterfaceCurve, &NiagaraDataInterfaceVectorCurve, &NiagaraDataInterfaceArrayFloat3]
 usedDynamicInputs [
  &SetBoolByFloatComparison
  &self:Abs_Z_Axis
  &FN_FrontEndFloatModifier
  &Multiply_Vector
 ]
 usedModules [
  &SystemState
  &self:SystemSpawnScript.SystemScriptSource.SystemScriptGraph.NiagaraNodeAssignment_2.SetVariables_#1
  &Timeline
  &self:SystemSpawnScript.SystemScriptSource.SystemScriptGraph.NiagaraNodeAssignment_1.SetVariables_#2
  &self:SystemSpawnScript.SystemScriptSource.SystemScriptGraph.NiagaraNodeAssignment_0.SetVariables_#3
 ]
}
emitters {
 Feet_Ribbons {
  emitterUpdateScript [
   &EmitterState()
   &SpawnBurst_Instantaneous(Spawn Count=&Multiply_Int(A=10, B=4), Use Spawn Probability=false, UseLoopCountLimit=true)
  ]
  particleSpawnScript [
   &PartitionParticles(Write Partition to Ribbon ID=::NewEnumerator0)
   /SetVariables(Particles.RibbonLinkOrder:NiagaraFloat=&ReturnNormalizedExecIndex()) {isSetParameters=true, module=SetVariables_#4}
   &InitializeParticle(
    Color Mode=::NewEnumerator0
    Lifetime=5
    Position Offset=<0 0 25>
    Ribbon Width=30
    Ribbon Width Mode=::NewEnumerator1
    Sprite Size Mode=::NewEnumerator3
    Uniform Sprite Size=5
    UsePositionOffset=false
    Write Material Random=true
   )
   &SkeletalMeshLocation(
    Bone / Socket Index=$Particles.PartitionParticles.Partition
    Bone Sampling Mode=::NewEnumerator4
    MaskSpawnGroup=false
    Sampled Position Offset Coordinate Space:ENiagaraSkelMeshTransforms=::NewEnumerator2
    Skeletal Mesh=$User.SkeletalMesh2
   )
  ]
  particleUpdateScript [
   /SetVariables(
    Particles.Ribbon2_Color:LinearColor=&MakeLinearColorFromVectorAndFloat("Vector (RGB)"=<2 2 2>, "Float (Alpha)"=&ClampFloat(Float=1))
    Particles.Ribbon3_Color:LinearColor=&MakeLinearColorFromVectorAndFloat("Vector (RGB)"=<3 1.0606 0.9844>, "Float (Alpha)"=&ClampFloat(Float=1))
    Particles.Ribbon3_DynamParam:Vector4f=<0.4 5 1 1>
   ) {isSetParameters=true, module=SetVariables_#5}
   &ParticleState(
    Kill Particles When Lifetime Has Elapsed=false
    Let Infinitely Lived Particles Die When Emitter Deactivates=true
    Loop Particles Lifetime=true
   )
   /Niagara/Modules/Update/Utility/IncrementOverTime(Rate of Change:NiagaraFloat=&Lerp_Float(A=1, B=1.5, Alpha=$User.ParaGlideLeanAlpha))
   &DynamicMaterialParameters(
    Index 0 Param 1=0.3
    Index 0 Param 2=5
    Index 0 Param 3=1
    Index 0 Param 4=1
    Module.Param0WriteEnabled=true
    Module.Param1WriteEnabled=true
    Module.Param2WriteEnabled=true
    Module.Param3WriteEnabled=true
   )
   &CurlNoiseForce(
    Noise Frequency=25
    Noise Strength=&Multiply_Float(A=1, B=$Particles.RibbonLinkOrder)
    Pan Noise=true
    Pan Noise Field=<0 0 2>
   )
   &SkeletalMeshLocation(
    Bone / Socket Index=$Particles.PartitionParticles.Partition
    Bone Sampling Mode=::NewEnumerator4
    MaskSpawnGroup=false
    Sampled Position Offset Coordinate Space:ENiagaraSkelMeshTransforms=::NewEnumerator1
    Skeletal Mesh=$User.SkeletalMesh2
   ) {module=SkeletalMeshLocation001}
   /SetVariables(Particles.Position=&AddVectorToPosition(
     Position=$Particles.Position
     Vector=&Multiply_VectorByFloat(
      Vector=$System.DiveVelocity
      Float=&FloatFromCurve(
       FloatCurve=DataInterface() {propertyValues {Curve {keys [{tangentMode=RCTM_User, time=0, value=0}, {tangentMode=RCTM_User, time=1, value=1}]}}}
       CurveIndex=&ReturnNormalizedExecIndex()
       Scale Curve=-150
      )
     )
    )) {isSetParameters=true, module=SetVariables_#6}
   /SetVariables(Particles.Position=&AddVectorToPosition(
     Position=$Particles.Position
     Vector=&TransformVector(
      Vector=&Multiply_VectorByFloat(
       Vector=&SelectVectorFromArray(
        Direct Array Index=$Particles.PartitionParticles.Partition
        Vector Selection Array=DataInterface() {propertyValues {FloatData [<0.75 0 0>, <-0.75 0 0>, <0.75 0 0>, <-0.75 0 0>]}}
       )
       Float=&FloatFromCurve(
        FloatCurve=DataInterface() {propertyValues {Curve {keys [
            {arriveTangent=4.1785, interpMode=RCIM_Cubic, leaveTangent=4.1785, tangentMode=RCTM_User, time=0, value=0}
            {arriveTangent=1.1179, interpMode=RCIM_Cubic, leaveTangent=1.1179, tangentMode=RCTM_User, time=0.2042, value=0.6228}
            {arriveTangent=0.0224, interpMode=RCIM_Cubic, leaveTangent=0.0224, tangentMode=RCTM_User, time=1, value=1}
           ]}}}
        CurveIndex=&ReturnNormalizedExecIndex()
        Scale Curve=&MakeCustomFloatFromBool(
         Bool=&SetBoolByIntComparison(A=$Particles.PartitionParticles.Partition, B=1, Comparison Type=::NewEnumerator0)
         True Float=6
         False Float=13
        )
       )
      )
      Destination Space=::Simulation
     )
    )) {enabled=false, isSetParameters=true, module=SetVariables_#7}
   /SetVariables(Particles.Position=&AddVectorToPosition(
     Position=$Particles.Position
     Vector=&Multiply_VectorByFloat(
      Vector=&NormalizeVector(Vector To Normalize=$Output.CurlNoiseForce.CurlNoiseForce, Safe Normalize=true, Safe Fallback Threshold=0, Safe Fallback Vector=<0 0 0.001>)
      Float=&FloatFromCurve(
       FloatCurve=DataInterface() {propertyValues {Curve {keys [{tangentMode=RCTM_User, time=0, value=0}, {tangentMode=RCTM_User, time=1, value=1}]}}}
       CurveIndex=&ReturnNormalizedExecIndex()
       Scale Curve=&Lerp_Float(A=1, B=2, Alpha=$User.ParaGlideLeanAlpha)
      )
     )
    )) {enabled=false, isSetParameters=true, module=SetVariables_#8}
   &ScaleRibbonWidth(Ribbon Width Scale=1.15)
  ]
  rendererClasses [&NiagaraSpriteRendererProperties, &NiagaraRibbonRendererProperties]
  renderers [
   {rendererClass=&NiagaraSpriteRendererProperties}
   {rendererClass=&NiagaraRibbonRendererProperties, rendererIndex=1}
   {rendererClass=&NiagaraRibbonRendererProperties, rendererIndex=2}
   {rendererClass=&NiagaraRibbonRendererProperties, rendererIndex=3}
   {rendererClass=&NiagaraRibbonRendererProperties, rendererIndex=4}
   {rendererClass=&NiagaraRibbonRendererProperties, rendererIndex=5}
  ]
 }
 Glow {
  emitterUpdateScript [
   &EmitterState(Inactive Response=::NewEnumerator1, Life Cycle Mode=::NewEnumerator1, Loop Duration=0.125, UseLoopDelay=true)
   &SpawnBurst_Instantaneous(
    Spawn Count=&RandomRangeInt(Minimum=0, Maximum=1)
    Use Spawn Probability=false
    UseLoopCountLimit=false
   )
   &SpawnRate(SpawnRate=4, Use Spawn Probability=false)
  ]
  particleSpawnScript [
   &PartitionParticles()
   &InitializeParticle(Lifetime=0.6, Sprite Size=<75 75>, Sprite Size Mode=::NewEnumerator4, UsePositionOffset=false, Write Material Random=true)
   &Color(Color=&MakeLinearColorFromVectorAndFloat("Vector (RGB)"=<0.724 0.724 0.724>, "Float (Alpha)"=&ClampFloat(Float=1)))
  ]
  particleUpdateScript [
   &SkeletalMeshLocation(
    Bone / Socket Index=$Particles.PartitionParticles.Partition
    Bone Sampling Mode=::NewEnumerator4
    MaskSpawnGroup=false
    Skeletal Mesh=$User.SkeletalMesh
   )
   &ParticleState()
   &ScaleSpriteSize()
   &CameraOffset(Camera Offset Amount=&RandomRangeFloat(Minimum=25, Maximum=35, Evaluation Type=::NewEnumerator1))
   &ScaleColor(
    Scale Alpha=&FloatFromCurve(FloatCurve=DataInterface() {propertyValues {Curve {keys [
         <at 0 0 RCIM_Cubic>
         {arriveTangent=1.2172, interpMode=RCIM_Cubic, leaveTangent=1.2172, time=0.0749, value=1}
         {arriveTangent=-1.081, interpMode=RCIM_Cubic, leaveTangent=-1.081, time=0.8216, value=1}
         <at 1 0>
        ]}}})
    Scale RGB=&VectorFromCurve(VectorCurve=DataInterface() {propertyValues {
       XCurve {keys [<at 0 1 RCIM_Cubic>, {arriveTangent=-0.2906, interpMode=RCIM_Cubic, leaveTangent=-0.2906, time=0.3, value=1}, <at 0.6 0.8256>, <at 1 1>]}
       YCurve {keys [
         <at 0 0.4201 RCIM_Cubic>
         {arriveTangent=-0.7002, interpMode=RCIM_Cubic, leaveTangent=-0.7002, time=0.3, value=0.1612}
         <at 0.6 0>
         <at 1 0.4201>
        ]}
       ZCurve {keys [<at 0 0 RCIM_Cubic>, {arriveTangent=1.6667, interpMode=RCIM_Cubic, leaveTangent=1.6667, time=0.3, value=0}, <at 0.6 1>, <at 1 0>]}
      }})
    ScaleA=true
    ScaleRGB=true
   )
  ]
  rendererClasses [&NiagaraSpriteRendererProperties]
  renderers [{rendererClass=&NiagaraSpriteRendererProperties}]
 }
 Hand_Ribbons {
  emitterUpdateScript [
   &EmitterState()
   &SpawnBurst_Instantaneous(Spawn Count=&Multiply_Int(A=10, B=4), Use Spawn Probability=false, UseLoopCountLimit=true)
  ]
  particleSpawnScript [
   &PartitionParticles(Write Partition to Ribbon ID=::NewEnumerator0)
   /SetVariables(Particles.RibbonLinkOrder:NiagaraFloat=&ReturnNormalizedExecIndex()) {isSetParameters=true, module=SetVariables_#4}
   &InitializeParticle(
    Color=<rgba 2 2 2 2>
    Lifetime=5
    Position Offset=<0 0 25>
    Ribbon Width=30
    Ribbon Width Mode=::NewEnumerator1
    Sprite Size Mode=::NewEnumerator3
    Uniform Sprite Size=5
    UsePositionOffset=false
    Write Material Random=true
   )
   &SkeletalMeshLocation(
    Bone / Socket Index=$Particles.PartitionParticles.Partition
    Bone Sampling Mode=::NewEnumerator4
    MaskSpawnGroup=false
    Sampled Position Offset Coordinate Space:ENiagaraSkelMeshTransforms=::NewEnumerator2
    Skeletal Mesh=$User.SkeletalMesh
   )
  ]
  particleUpdateScript [
   /SetVariables(
    Particles.Ribbon2_Color:LinearColor=&MakeLinearColorFromVectorAndFloat("Vector (RGB)"=<2 2 2>, "Float (Alpha)"=&ClampFloat(Float=1))
    Particles.Ribbon3_Color:LinearColor=&MakeLinearColorFromVectorAndFloat("Vector (RGB)"=<3 1.0606 0.9844>, "Float (Alpha)"=&ClampFloat(Float=1))
    Particles.Ribbon3_DynamParam:Vector4f=<0.4 5 1 1>
   ) {isSetParameters=true, module=SetVariables_#5}
   &ParticleState(
    Kill Particles When Lifetime Has Elapsed=false
    Let Infinitely Lived Particles Die When Emitter Deactivates=true
    Loop Particles Lifetime=true
   )
   &IncrementOverTime(Rate of Change:NiagaraFloat=&Lerp_Float(A=1, B=1.5, Alpha=$User.ParaGlideLeanAlpha))
   &DynamicMaterialParameters(
    Index 0 Param 1=0.3
    Index 0 Param 2=5
    Index 0 Param 3=1
    Index 0 Param 4=1
    Module.Param0WriteEnabled=true
    Module.Param1WriteEnabled=true
    Module.Param2WriteEnabled=true
    Module.Param3WriteEnabled=true
   )
   &CurlNoiseForce(
    Noise Frequency=25
    Noise Strength=&Multiply_Float(A=1, B=$Particles.RibbonLinkOrder)
    Pan Noise=true
    Pan Noise Field=<0 0 2>
   )
   &SkeletalMeshLocation(
    Bone / Socket Index=$Particles.PartitionParticles.Partition
    Bone Sampling Mode=::NewEnumerator4
    MaskSpawnGroup=false
    Sampled Position Offset Coordinate Space:ENiagaraSkelMeshTransforms=::NewEnumerator1
    Skeletal Mesh=$User.SkeletalMesh
   ) {module=SkeletalMeshLocation001}
   /SetVariables(Particles.Position=&AddVectorToPosition(
     Position=$Particles.Position
     Vector=&Multiply_VectorByFloat(
      Vector=$System.DiveVelocity
      Float=&FloatFromCurve(
       FloatCurve=DataInterface() {propertyValues {Curve {keys [{tangentMode=RCTM_User, time=0, value=0}, {tangentMode=RCTM_User, time=1, value=1}]}}}
       CurveIndex=&ReturnNormalizedExecIndex()
       Scale Curve=-200
      )
     )
    )) {isSetParameters=true, module=SetVariables_#6}
   /SetVariables(Particles.Position=&AddVectorToPosition(
     Position=$Particles.Position
     Vector=&TransformVector(
      Vector=&Multiply_VectorByFloat(
       Vector=&SelectVectorFromArray(
        Direct Array Index=$Particles.PartitionParticles.Partition
        Vector Selection Array=DataInterface() {propertyValues {FloatData [<0.75 0 0>, <-0.75 0 0>, <0.75 0 0>, <-0.75 0 0>]}}
       )
       Float=&FloatFromCurve(
        FloatCurve=DataInterface() {propertyValues {Curve {keys [
            {arriveTangent=4.1785, interpMode=RCIM_Cubic, leaveTangent=4.1785, tangentMode=RCTM_User, time=0, value=0}
            {arriveTangent=1.1179, interpMode=RCIM_Cubic, leaveTangent=1.1179, tangentMode=RCTM_User, time=0.2042, value=0.6228}
            {arriveTangent=0.0224, interpMode=RCIM_Cubic, leaveTangent=0.0224, tangentMode=RCTM_User, time=1, value=1}
           ]}}}
        CurveIndex=&ReturnNormalizedExecIndex()
        Scale Curve=&MakeCustomFloatFromBool(
         Bool=&SetBoolByIntComparison(A=$Particles.PartitionParticles.Partition, B=1, Comparison Type=::NewEnumerator0)
         True Float=1
         False Float=1
        )
       )
      )
      Destination Space=::Simulation
     )
    )) {enabled=false, isSetParameters=true, module=SetVariables_#7}
   /SetVariables(Particles.Position=&AddVectorToPosition(
     Position=$Particles.Position
     Vector=&Multiply_VectorByFloat(
      Vector=&NormalizeVector(Vector To Normalize=$Output.CurlNoiseForce.CurlNoiseForce, Safe Normalize=true, Safe Fallback Threshold=0, Safe Fallback Vector=<0 0 0.001>)
      Float=&FloatFromCurve(
       FloatCurve=DataInterface() {propertyValues {Curve {keys [{tangentMode=RCTM_User, time=0, value=0}, {tangentMode=RCTM_User, time=1, value=1}]}}}
       CurveIndex=&ReturnNormalizedExecIndex()
       Scale Curve=&Lerp_Float(A=1, B=2, Alpha=$User.ParaGlideLeanAlpha)
      )
     )
    )) {enabled=false, isSetParameters=true, module=SetVariables_#8}
   &ScaleRibbonWidth(Ribbon Width Scale=1.15)
  ]
  rendererClasses [&NiagaraSpriteRendererProperties, &NiagaraRibbonRendererProperties]
  renderers [
   {rendererClass=&NiagaraSpriteRendererProperties}
   {rendererClass=&NiagaraRibbonRendererProperties, rendererIndex=1}
   {rendererClass=&NiagaraRibbonRendererProperties, rendererIndex=2}
   {rendererClass=&NiagaraRibbonRendererProperties, rendererIndex=3}
   {rendererClass=&NiagaraRibbonRendererProperties, rendererIndex=4}
   {rendererClass=&NiagaraRibbonRendererProperties, rendererIndex=5}
  ]
 }
 Stars {
  emitterUpdateScript [
   &EmitterState(Inactive Response=::NewEnumerator1, Life Cycle Mode=::NewEnumerator1, Loop Duration=1, UseLoopDelay=true)
   &SpawnRate(SpawnRate=100, Use Spawn Probability=false)
  ]
  particleSpawnScript [
   &PartitionParticles()
   &InitializeParticle(
    Color=<rgba 2 2 2 1>
    Lifetime=&RandomRangeFloat(Minimum=0.1, Maximum=0.2)
    Sprite Size=&Vector2DFromFloat(Value=&RandomRangeFloat(Minimum=5, Maximum=30))
    Sprite Size Mode=::NewEnumerator4
    UsePositionOffset=false
    Write Material Random=true
   )
   &SkeletalMeshLocation(
    Bone / Socket Index=$Particles.PartitionParticles.Partition
    Bone Sampling Mode=::NewEnumerator4
    MaskSpawnGroup=false
    Skeletal Mesh=$User.SkeletalMesh
   )
   &Color(Color=&MakeLinearColorFromVectorAndFloat("Vector (RGB)"=<3 3 3>, "Float (Alpha)"=&ClampFloat(Float=1)))
   &ShapeLocation(
    Cylinder Height=200
    Cylinder Height Midpoint=-0.1
    Cylinder Radius=10
    MaskSpawnGroup=false
    Offset=&FN_FrontEndVector3Modifier(Gameplay=<0 0 0>, Front End=&Lerp_Vector(A=<0 0 0>, B=<0 -20 0>, Alpha=$System.Dive))
    Offset Mode=::NewEnumerator0
    Override Seed=false
    Rotation Axis=<0 0 1>
    Rotation Mode=::NewEnumerator1
    Shape Primitive=::NewEnumerator1
    Surface Only=true
    UseOwnerScale=true
   )
   &DynamicMaterialParameters(
    Index 0 Param 1=1
    Module.Param0WriteEnabled=true
    Module.Param1WriteEnabled=false
    Module.Param2WriteEnabled=false
    Module.Param3WriteEnabled=false
   )
  ]
  particleUpdateScript [
   &ParticleState()
   &SkeletalMeshLocation(
    Bone / Socket Index=$Particles.PartitionParticles.Partition
    Bone Sampling Mode=::NewEnumerator4
    MaskSpawnGroup=false
    Skeletal Mesh=$User.SkeletalMesh
   ) {enabled=false, module=SkeletalMeshLocation001}
   &ScaleSpriteSize()
   &DynamicMaterialParameters(
    Index 0 Param 2=&RandomRangeFloat(Minimum=0, Maximum=1, Evaluation Type=::NewEnumerator1)
    Index 0 Param 3=200
    Index 0 Param 4=1
    Module.Param0WriteEnabled=false
    Module.Param1WriteEnabled=true
    Module.Param2WriteEnabled=true
    Module.Param3WriteEnabled=true
   ) {module=DynamicMaterialParameters001}
  ]
  rendererClasses [&NiagaraSpriteRendererProperties]
  renderers [{rendererClass=&NiagaraSpriteRendererProperties}]
 }
}
summary {
 systemName=NS_Contrail_Kpop_Converted
 emitters [
  {emitterName=Stars, rendererClasses [&NiagaraSpriteRendererProperties]}
  {emitterName=Glow, rendererClasses [&NiagaraSpriteRendererProperties]}
  {emitterName=Hand_Ribbons, rendererClasses [&NiagaraSpriteRendererProperties, &NiagaraRibbonRendererProperties]}
  {emitterName=Feet_Ribbons, rendererClasses [&NiagaraSpriteRendererProperties, &NiagaraRibbonRendererProperties]}
 ]
 userVariables [
  User.bIsFrontEnd {type=NiagaraBool}
  User.bIsFrontEndPreview {type=NiagaraBool, defaultValue=true}
  User.ParaGlideLeanAlpha {type=NiagaraFloat, defaultValue=0, description=""}
  User.PlayerSpeed {type=NiagaraFloat, defaultValue=0, description=""}
  User.SkeletalMesh {
   type=NiagaraDataInterfaceSkeletalMesh
   defaultValue {
    struct=NiagaraExt_VariableValue_DataInterface
    value {dataInterface=&self:NiagaraDataInterfaceSkeletalMesh_0}
   }
  }
  User.SkeletalMesh2 {
   type=NiagaraDataInterfaceSkeletalMesh
   defaultValue {
    struct=NiagaraExt_VariableValue_DataInterface
    value {dataInterface=&self:NiagaraDataInterfaceSkeletalMesh_2}
   }
  }
 ]
}
